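# Control Rig for the Parcel Petal back bling, procedurally animating a plant with stems, tongue, petals and two chomping heads.
# Chain Harmonics waves sway three stems and the tongue.
# Scaled, accumulated slerps of those bones' rotations drive secondary stems, while sines of random-speed timelines chomp the jaws and heads.
# The didKill variable weights a fast world-time sine rotation over both jaws and the mid petals, presumably as a kill reaction.
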ControlRigBlueprint /BRCosmetics/Accessories/FORT_Backpacks/Backpack_F_MED_Parcel_Petal/Meshes/F_MED_Parcel_Petal_Pack_CtrlRig/F_MED_Parcel_Petal_Pack_CtrlRig
graphs [
 RigVMFunctionLibrary {links [], nodes []}
 RigVMModel {nodes [
   RigUnit_BeginExecution = &RigUnit_BeginExecution()
   /Script/ControlRig.RigUnit_ChainHarmonics(
    ChainRoot=stem_top_mid_01
    DrawWorldOffset=(Rotation=(X=0,Y=0,Z=0,W=1),Translation=(X=0,Y=0,Z=0),Scale3D=(X=1,Y=1,Z=1))
    Pendulum=(bEnabled=false,PendulumStiffness=2,PendulumGravity=(X=0,Y=0,Z=0),PendulumBlend=0.75,PendulumDrag=0.98,PendulumMaximum=0.1,UnwindAxis=(X=0,Y=1,Z=0),UnwindMinimum=0.2,UnwindMaximum=0.05)
    Reach=(bEnabled=True,ReachTarget=(X=0,Y=0,Z=0),ReachAxis=(X=1,Y=0,Z=0))
    Speed=(X=1,Y=1,Z=1)
    Wave=(bEnabled=True,WaveFrequency=(X=100,Y=100,Z=100),WaveAmplitude=(X=0,Y=0.5,Z=0.5),WaveOffset=(X=0,Y=2,Z=1),WaveNoise=(X=0,Y=1,Z=0),WaveMinimum=0,WaveMaximum=1)
    WaveCurve=(EditorCurveData=(Keys=((Value=0.25),(Time=1,Value=0.75)),DefaultValue=340282346638528859811704183484516925440,PreInfinityExtrap=RCCE_Constant,PostInfinityExtrap=RCCE_Constant))
    bDrawDebug=True
    ExecutePin=@RigUnit_BeginExecution.ExecutePin
   ) {is=struct RigVMUnitNode}
   > &RigUnit_ChainHarmonics(
    ChainRoot=stem_r_01
    DrawWorldOffset=(Rotation=(X=0,Y=0,Z=0,W=1),Translation=(X=0,Y=0,Z=0),Scale3D=(X=1,Y=1,Z=1))
    Pendulum=(bEnabled=false,PendulumStiffness=2,PendulumGravity=(X=0,Y=0,Z=0),PendulumBlend=0.75,PendulumDrag=0.98,PendulumMinimum=0,PendulumMaximum=0.1,PendulumEase=Linear,UnwindAxis=(X=0,Y=1,Z=0),UnwindMinimum=0.2,UnwindMaximum=0.05)
    Reach=(bEnabled=false,ReachTarget=(X=10,Y=0,Z=10),ReachAxis=(X=1,Y=0,Z=0),ReachMinimum=0,ReachMaximum=-5,ReachEase=Linear)
    Speed=(X=1,Y=1,Z=1)
    Wave=(bEnabled=True,WaveFrequency=(X=100,Y=100,Z=100),WaveAmplitude=(X=0,Y=2,Z=2),WaveOffset=(X=0,Y=2,Z=1),WaveNoise=(X=0,Y=1,Z=0),WaveMinimum=0,WaveMaximum=0.4,WaveEase=SineEaseOut)
    WaveCurve=(EditorCurveData=(Keys=((Value=0.2),(Time=1,Value=0.75)),DefaultValue=340282346638528859811704183484516925440,PreInfinityExtrap=RCCE_Constant,PostInfinityExtrap=RCCE_Constant))
    bDrawDebug=True
   )
   > &RigUnit_ChainHarmonics(
    ChainRoot=stem_l_mid_01
    DrawWorldOffset=(Rotation=(X=0,Y=0,Z=0,W=1),Translation=(X=0,Y=0,Z=0),Scale3D=(X=1,Y=1,Z=1))
    Pendulum=(bEnabled=false,PendulumStiffness=2,PendulumGravity=(X=0,Y=0,Z=0),PendulumBlend=0.75,PendulumDrag=0.98,PendulumMinimum=0,PendulumMaximum=0.1,PendulumEase=Linear,UnwindAxis=(X=0,Y=1,Z=0),UnwindMinimum=0.2,UnwindMaximum=0.05)
    Reach=(bEnabled=false,ReachTarget=(X=10,Y=0,Z=10),ReachAxis=(X=1,Y=0,Z=0),ReachMinimum=0,ReachMaximum=-5,ReachEase=Linear)
    Speed=(X=1,Y=1,Z=1)
    Wave=(bEnabled=True,WaveFrequency=(X=100,Y=100,Z=100),WaveAmplitude=(X=0,Y=0.7,Z=0.7),WaveOffset=(X=0,Y=2,Z=1),WaveNoise=(X=0,Y=1,Z=0),WaveMinimum=0,WaveMaximum=0.4,WaveEase=SineEaseOut)
    WaveCurve=(EditorCurveData=(Keys=((Value=0.2),(Time=1,Value=0.75)),DefaultValue=340282346638528859811704183484516925440,PreInfinityExtrap=RCCE_Constant,PostInfinityExtrap=RCCE_Constant))
    bDrawDebug=True
   )
   > &RigUnit_ChainHarmonics(
    ChainRoot=tongue_01
    DrawWorldOffset=(Rotation=(X=0,Y=0,Z=0,W=1),Translation=(X=0,Y=0,Z=0),Scale3D=(X=1,Y=1,Z=1))
    Pendulum=(bEnabled=false,PendulumStiffness=2,PendulumGravity=(X=0,Y=0,Z=0),PendulumBlend=0.75,PendulumDrag=0.98,PendulumMinimum=0,PendulumMaximum=0.1,PendulumEase=Linear,UnwindAxis=(X=0,Y=1,Z=0),UnwindMinimum=0.2,UnwindMaximum=0.05)
    Reach=(bEnabled=false,ReachTarget=(X=10,Y=0,Z=10),ReachAxis=(X=1,Y=0,Z=0),ReachMinimum=0,ReachMaximum=-5,ReachEase=Linear)
    Speed=(X=1,Y=1,Z=1)
    Wave=(bEnabled=True,WaveFrequency=(X=75,Y=75,Z=-75),WaveAmplitude=(X=0.5,Y=1.5,Z=1),WaveOffset=(X=0,Y=1,Z=2),WaveNoise=(X=0,Y=1,Z=1),WaveMinimum=0,WaveMaximum=1,WaveEase=Linear)
    WaveCurve=(EditorCurveData=(Keys=((Value=0.25),(Time=1.0107,Value=0.6084)),DefaultValue=340282346638528859811704183484516925440,PreInfinityExtrap=RCCE_Constant,PostInfinityExtrap=RCCE_Constant))
    bDrawDebug=True
   )
   n1 = &RigUnit_GetBoneTransform(Bone=stem_r_03, Space=LocalSpace)
   > /Script/ControlRig.RigUnit_SetBoneRotation(
    Bone=chomp_stem_r_02
    Space=LocalSpace
    Weight=0.1
    bPropagateToChildren=True
    Rotation=/Script/RigVM.RigVMFunction_AccumulateQuatLerp(
     Blend=0.5
     InitialValue=(X=0,Y=0,Z=0,W=1)
     bIntegrateDeltaTime=false
     TargetValue=/Script/RigVM.RigVMFunction_MathQuaternionSlerp(A=(X=0,Y=0,Z=0,W=1), T=0.2, B=@n1.Transform.Rotation) {is=struct RigVMUnitNode, template="Interpolate::Execute(in A,in B,in T,out Result)"}.Result
    ) {is=struct RigVMUnitNode, template="AccumulateLerp::Execute(in TargetValue,in InitialValue,in Blend,in bIntegrateDeltaTime,out Result)"}.Result
   ) {is=struct RigVMUnitNode}
   > &RigUnit_SetBoneRotation(
    Bone=chomp_stem_r_03
    Weight=0.5
    bPropagateToChildren=True
    Rotation=&RigVMFunction_AccumulateQuatLerp(
     Blend=0.35
     InitialValue=(X=0,Y=0,Z=0,W=1)
     TargetValue=&RigVMFunction_MathQuaternionSlerp(A=(X=0,Y=0,Z=0,W=1), T=0.3, B=@n1.Transform.Rotation) {template="Interpolate::Execute(in A,in B,in T,out Result)"}.Result
    ) {template="AccumulateLerp::Execute(in TargetValue,in InitialValue,in Blend,in bIntegrateDeltaTime,out Result)"}.Result
   )
   > &RigUnit_SetBoneRotation(
    Bone=chomp_stem_r_04
    Weight=0.5
    bPropagateToChildren=True
    Rotation=&RigVMFunction_AccumulateQuatLerp(
     Blend=0.2
     InitialValue=(X=0,Y=0,Z=0,W=1)
     TargetValue=&RigVMFunction_MathQuaternionSlerp(A=(X=0,Y=0,Z=0,W=1), T=0.4, B=@n1.Transform.Rotation) {template="Interpolate::Execute(in A,in B,in T,out Result)"}.Result
    ) {template="AccumulateLerp::Execute(in TargetValue,in InitialValue,in Blend,in bIntegrateDeltaTime,out Result)"}.Result
   )
   n2 = /Script/RigVM.RigVMFunction_MathFloatSin(Value=/Script/RigVM.RigVMFunction_Timeline(Speed=/Script/RigVM.RigVMFunction_RandomFloat(Duration=1, Maximum=5, Seed=300) {is=struct RigVMUnitNode, template="Random::Execute(in Seed,in Minimum,in Maximum,in Duration,out Result)"}.Result) {is=struct RigVMUnitNode}.Time) {is=struct RigVMUnitNode, template="Sin::Execute(in Value,out Result)"}
   > &RigUnit_SetBoneRotation(
    Bone=chomp_stem_r_jaw
    Space=LocalSpace
    bPropagateToChildren=True
    Weight=@n2.Result
    Rotation=&RigVMFunction_MathQuaternionFromEuler(Euler=(X=0,Y=0,Z=80)).Result
   )
   > &RigUnit_SetBoneRotation(
    Bone=chomp_stem_r_head
    Space=LocalSpace
    bPropagateToChildren=True
    Weight=@n2.Result
    Rotation=&RigVMFunction_MathQuaternionFromEuler(Euler=(X=0,Y=0,Z=10)).Result
   )
   n3 = &RigUnit_GetBoneTransform(Bone=stem_r_03, Space=LocalSpace)
   > &RigUnit_SetBoneRotation(
    Bone=stem_top_r_01
    Space=LocalSpace
    Weight=0.1
    bPropagateToChildren=True
    Rotation=&RigVMFunction_AccumulateQuatLerp(
     Blend=1
     InitialValue=(X=0,Y=0,Z=0,W=1)
     bIntegrateDeltaTime=false
     TargetValue=&RigVMFunction_MathQuaternionSlerp(A=(X=0,Y=0,Z=0,W=1), T=0.2, B=@n3.Transform.Rotation) {template="Interpolate::Execute(in A,in B,in T,out Result)"}.Result
    ) {template="AccumulateLerp::Execute(in TargetValue,in InitialValue,in Blend,in bIntegrateDeltaTime,out Result)"}.Result
   )
   > &RigUnit_SetBoneRotation(
    Bone=stem_top_r_02
    Space=LocalSpace
    Weight=1
    bPropagateToChildren=True
    Rotation=&RigVMFunction_AccumulateQuatLerp(
     Blend=1
     InitialValue=(X=0,Y=0,Z=0,W=1)
     TargetValue=&RigVMFunction_MathQuaternionSlerp(A=(X=0,Y=0,Z=0,W=1), T=0.4, B=@n3.Transform.Rotation) {template="Interpolate::Execute(in A,in B,in T,out Result)"}.Result
    ) {template="AccumulateLerp::Execute(in TargetValue,in InitialValue,in Blend,in bIntegrateDeltaTime,out Result)"}.Result
   )
   > &RigUnit_SetBoneRotation(
    Bone=stem_top_r_03
    Space=LocalSpace
    Weight=1
    bPropagateToChildren=True
    Rotation=&RigVMFunction_AccumulateQuatLerp(
     Blend=1
     InitialValue=(X=0,Y=0,Z=0,W=1)
     TargetValue=&RigVMFunction_MathQuaternionSlerp(A=(X=0,Y=0,Z=0,W=1), T=0.65, B=@n3.Transform.Rotation) {template="Interpolate::Execute(in A,in B,in T,out Result)"}.Result
    ) {template="AccumulateLerp::Execute(in TargetValue,in InitialValue,in Blend,in bIntegrateDeltaTime,out Result)"}.Result
   )
   n4 = &RigUnit_GetBoneTransform(Bone=stem_l_mid_03, Space=LocalSpace)
   > &RigUnit_SetBoneRotation(
    Bone=stem_top_l_01
    Space=LocalSpace
    Weight=0.1
    bPropagateToChildren=True
    Rotation=&RigVMFunction_AccumulateQuatLerp(
     Blend=1
     InitialValue=(X=0,Y=0,Z=0,W=1)
     bIntegrateDeltaTime=false
     TargetValue=&RigVMFunction_MathQuaternionSlerp(A=(X=0,Y=0,Z=0,W=1), T=0.2, B=@n4.Transform.Rotation) {template="Interpolate::Execute(in A,in B,in T,out Result)"}.Result
    ) {template="AccumulateLerp::Execute(in TargetValue,in InitialValue,in Blend,in bIntegrateDeltaTime,out Result)"}.Result
   )
   > &RigUnit_SetBoneRotation(
    Bone=stem_top_l_02
    Space=LocalSpace
    Weight=1
    bPropagateToChildren=True
    Rotation=&RigVMFunction_AccumulateQuatLerp(
     Blend=1
     InitialValue=(X=0,Y=0,Z=0,W=1)
     TargetValue=&RigVMFunction_MathQuaternionSlerp(A=(X=0,Y=0,Z=0,W=1), T=0.4, B=@n4.Transform.Rotation) {template="Interpolate::Execute(in A,in B,in T,out Result)"}.Result
    ) {template="AccumulateLerp::Execute(in TargetValue,in InitialValue,in Blend,in bIntegrateDeltaTime,out Result)"}.Result
   )
   > &RigUnit_SetBoneRotation(
    Bone=stem_top_l_03
    Space=LocalSpace
    Weight=1
    bPropagateToChildren=True
    Rotation=&RigVMFunction_AccumulateQuatLerp(
     Blend=1
     InitialValue=(X=0,Y=0,Z=0,W=1)
     TargetValue=&RigVMFunction_MathQuaternionSlerp(A=(X=0,Y=0,Z=0,W=1), T=0.65, B=@n4.Transform.Rotation) {template="Interpolate::Execute(in A,in B,in T,out Result)"}.Result
    ) {template="AccumulateLerp::Execute(in TargetValue,in InitialValue,in Blend,in bIntegrateDeltaTime,out Result)"}.Result
   )
   n5 = &RigUnit_GetBoneTransform(Bone=tongue_04, Space=LocalSpace)
   > &RigUnit_SetBoneRotation(
    Bone=stem_l_btm_01
    Space=LocalSpace
    Weight=0.1
    bPropagateToChildren=True
    Rotation=&RigVMFunction_AccumulateQuatLerp(
     Blend=1
     InitialValue=(X=0,Y=0,Z=0,W=1)
     bIntegrateDeltaTime=false
     TargetValue=&RigVMFunction_MathQuaternionSlerp(A=(X=0,Y=0,Z=0,W=1), T=0.2, B=@n5.Transform.Rotation) {template="Interpolate::Execute(in A,in B,in T,out Result)"}.Result
    ) {template="AccumulateLerp::Execute(in TargetValue,in InitialValue,in Blend,in bIntegrateDeltaTime,out Result)"}.Result
   )
   > &RigUnit_SetBoneRotation(
    Bone=stem_l_btm_02
    Space=LocalSpace
    Weight=0.5
    bPropagateToChildren=True
    Rotation=&RigVMFunction_AccumulateQuatLerp(
     Blend=0.35
     InitialValue=(X=0,Y=0,Z=0,W=1)
     TargetValue=&RigVMFunction_MathQuaternionSlerp(A=(X=0,Y=0,Z=0,W=1), T=0.4, B=@n5.Transform.Rotation) {template="Interpolate::Execute(in A,in B,in T,out Result)"}.Result
    ) {template="AccumulateLerp::Execute(in TargetValue,in InitialValue,in Blend,in bIntegrateDeltaTime,out Result)"}.Result
   )
   > &RigUnit_SetBoneRotation(
    Bone=stem_l_btm_03
    Space=LocalSpace
    Weight=0.1
    bPropagateToChildren=True
    Rotation=&RigVMFunction_AccumulateQuatLerp(
     Blend=0.2
     InitialValue=(X=0,Y=0,Z=0,W=1)
     TargetValue=&RigVMFunction_MathQuaternionSlerp(A=(X=0,Y=0,Z=0,W=1), T=0.65, B=@n5.Transform.Rotation) {template="Interpolate::Execute(in A,in B,in T,out Result)"}.Result
    ) {template="AccumulateLerp::Execute(in TargetValue,in InitialValue,in Blend,in bIntegrateDeltaTime,out Result)"}.Result
   )
   > &RigUnit_SetBoneRotation(
    Bone=stem_l_btm_04
    Space=LocalSpace
    Weight=0.5
    bPropagateToChildren=True
    Rotation=&RigVMFunction_AccumulateQuatLerp(
     Blend=0.2
     InitialValue=(X=0,Y=0,Z=0,W=1)
     TargetValue=&RigVMFunction_MathQuaternionSlerp(A=(X=0,Y=0,Z=0,W=1), T=0.65, B=@n5.Transform.Rotation) {template="Interpolate::Execute(in A,in B,in T,out Result)"}.Result
    ) {template="AccumulateLerp::Execute(in TargetValue,in InitialValue,in Blend,in bIntegrateDeltaTime,out Result)"}.Result
   )
   n6 = &RigUnit_GetBoneTransform(Bone=tongue_04, Space=LocalSpace)
   > &RigUnit_SetBoneRotation(
    Bone=chomp_stem_l_01
    Space=LocalSpace
    Weight=0.1
    bPropagateToChildren=True
    Rotation=&RigVMFunction_AccumulateQuatLerp(
     Blend=1
     InitialValue=(X=0,Y=0,Z=0,W=1)
     bIntegrateDeltaTime=false
     TargetValue=&RigVMFunction_MathQuaternionSlerp(A=(X=0,Y=0,Z=0,W=1), T=0.2, B=@n6.Transform.Rotation) {template="Interpolate::Execute(in A,in B,in T,out Result)"}.Result
    ) {template="AccumulateLerp::Execute(in TargetValue,in InitialValue,in Blend,in bIntegrateDeltaTime,out Result)"}.Result
   )
   > &RigUnit_SetBoneRotation(
    Bone=chomp_stem_l_02
    Space=LocalSpace
    Weight=0.1
    bPropagateToChildren=True
    Rotation=&RigVMFunction_AccumulateQuatLerp(
     Blend=0.35
     InitialValue=(X=0,Y=0,Z=0,W=1)
     TargetValue=&RigVMFunction_MathQuaternionSlerp(A=(X=0,Y=0,Z=0,W=1), T=0.4, B=@n6.Transform.Rotation) {template="Interpolate::Execute(in A,in B,in T,out Result)"}.Result
    ) {template="AccumulateLerp::Execute(in TargetValue,in InitialValue,in Blend,in bIntegrateDeltaTime,out Result)"}.Result
   )
   > &RigUnit_SetBoneRotation(
    Bone=chomp_stem_l_03
    Space=LocalSpace
    Weight=0.5
    bPropagateToChildren=True
    Rotation=&RigVMFunction_AccumulateQuatLerp(
     Blend=0.2
     InitialValue=(X=0,Y=0,Z=0,W=1)
     TargetValue=&RigVMFunction_MathQuaternionSlerp(A=(X=0,Y=0,Z=0,W=1), T=0.65, B=@n6.Transform.Rotation) {template="Interpolate::Execute(in A,in B,in T,out Result)"}.Result
    ) {template="AccumulateLerp::Execute(in TargetValue,in InitialValue,in Blend,in bIntegrateDeltaTime,out Result)"}.Result
   )
   n7 = &RigVMFunction_MathFloatSin(Value=&RigVMFunction_Timeline(Speed=&RigVMFunction_RandomFloat(Duration=1, Maximum=5, Minimum=1, Seed=217) {template="Random::Execute(in Seed,in Minimum,in Maximum,in Duration,out Result)"}.Result).Time) {template="Sin::Execute(in Value,out Result)"}
   > &RigUnit_SetBoneRotation(
    Bone=chomp_stem_l_jaw
    Space=LocalSpace
    bPropagateToChildren=True
    Weight=@n7.Result
    Rotation=&RigVMFunction_MathQuaternionFromEuler(Euler=(X=0,Y=0,Z=50)).Result
   )
   > &RigUnit_SetBoneRotation(
    Bone=chomp_stem_l_head
    Space=LocalSpace
    bPropagateToChildren=True
    Weight=@n7.Result
    Rotation=&RigVMFunction_MathQuaternionFromEuler(Euler=(X=0,Y=0,Z=10)).Result
   )
   n8 = &RigUnit_GetBoneTransform(Bone=tongue_04, Space=LocalSpace)
   > &RigUnit_SetBoneRotation(
    Bone=stem_r_btm_01
    Space=LocalSpace
    Weight=0.2
    bPropagateToChildren=True
    Rotation=&RigVMFunction_AccumulateQuatLerp(
     Blend=1
     InitialValue=(X=0,Y=0,Z=0,W=1)
     bIntegrateDeltaTime=false
     TargetValue=&RigVMFunction_MathQuaternionSlerp(A=(X=0,Y=0,Z=0,W=1), T=0.2, B=@n8.Transform.Rotation) {template="Interpolate::Execute(in A,in B,in T,out Result)"}.Result
    ) {template="AccumulateLerp::Execute(in TargetValue,in InitialValue,in Blend,in bIntegrateDeltaTime,out Result)"}.Result
   )
   > &RigUnit_SetBoneRotation(
    Bone=stem_r_btm_02
    Space=LocalSpace
    Weight=0.5
    bPropagateToChildren=True
    Rotation=&RigVMFunction_AccumulateQuatLerp(
     Blend=1
     InitialValue=(X=0,Y=0,Z=0,W=1)
     TargetValue=&RigVMFunction_MathQuaternionSlerp(A=(X=0,Y=0,Z=0,W=1), T=0.4, B=@n8.Transform.Rotation) {template="Interpolate::Execute(in A,in B,in T,out Result)"}.Result
    ) {template="AccumulateLerp::Execute(in TargetValue,in InitialValue,in Blend,in bIntegrateDeltaTime,out Result)"}.Result
   )
   > &RigUnit_SetBoneRotation(
    Bone=stem_r_btm_03
    Space=LocalSpace
    Weight=0.5
    bPropagateToChildren=True
    Rotation=&RigVMFunction_AccumulateQuatLerp(
     Blend=1
     InitialValue=(X=0,Y=0,Z=0,W=1)
     TargetValue=&RigVMFunction_MathQuaternionSlerp(A=(X=0,Y=0,Z=0,W=1), T=0.65, B=@n8.Transform.Rotation) {template="Interpolate::Execute(in A,in B,in T,out Result)"}.Result
    ) {template="AccumulateLerp::Execute(in TargetValue,in InitialValue,in Blend,in bIntegrateDeltaTime,out Result)"}.Result
   )
   > &RigUnit_SetBoneRotation(
    Bone=stem_r_btm_04
    Space=LocalSpace
    Weight=0.1
    bPropagateToChildren=True
    Rotation=&RigVMFunction_AccumulateQuatLerp(
     Blend=1
     InitialValue=(X=0,Y=0,Z=0,W=1)
     TargetValue=&RigVMFunction_MathQuaternionSlerp(A=(X=0,Y=0,Z=0,W=1), T=0.65, B=@n8.Transform.Rotation) {template="Interpolate::Execute(in A,in B,in T,out Result)"}.Result
    ) {template="AccumulateLerp::Execute(in TargetValue,in InitialValue,in Blend,in bIntegrateDeltaTime,out Result)"}.Result
   )
   > &RigUnit_SetBoneRotation(
    Bone=stem_r_05
    Space=LocalSpace
    Weight=0.5
    bPropagateToChildren=True
    Rotation=&RigVMFunction_AccumulateQuatLerp(
     Blend=1
     InitialValue=(X=0,Y=0,Z=0,W=1)
     TargetValue=&RigVMFunction_MathQuaternionSlerp(A=(X=0,Y=0,Z=0,W=1), T=0.65, B=@n8.Transform.Rotation) {template="Interpolate::Execute(in A,in B,in T,out Result)"}.Result
    ) {template="AccumulateLerp::Execute(in TargetValue,in InitialValue,in Blend,in bIntegrateDeltaTime,out Result)"}.Result
   )
   > &RigUnit_SetBoneRotation(
    Bone=stem_r_06
    Space=LocalSpace
    Weight=0.8
    bPropagateToChildren=True
    Rotation=&RigVMFunction_AccumulateQuatLerp(
     Blend=1
     InitialValue=(X=0,Y=0,Z=0,W=1)
     TargetValue=&RigVMFunction_MathQuaternionSlerp(A=(X=0,Y=0,Z=0,W=1), T=0.5, B=@n8.Transform.Rotation) {template="Interpolate::Execute(in A,in B,in T,out Result)"}.Result
    ) {template="AccumulateLerp::Execute(in TargetValue,in InitialValue,in Blend,in bIntegrateDeltaTime,out Result)"}.Result
   )
   n9 = &RigVMFunction_MathFloatSin(Value=&RigVMFunction_MathFloatMul(B=15, A=/Script/RigVM.RigVMFunction_GetWorldTime() {is=struct RigVMUnitNode}.Seconds) {template="Multiply::Execute(in A,in B,out Result)"}.Result) {template="Sin::Execute(in Value,out Result)"}
   n10 = &RigVMFunction_MathQuaternionFromEuler(
    Euler=(X=0,Y=0,Z=0)
    Euler.Z=&RigVMFunction_MathFloatAdd(B=50, A=&RigVMFunction_MathFloatMul(B=25, A=@n9.Result) {template="Multiply::Execute(in A,in B,out Result)"}.Result) {template="Add::Execute(in A,in B,out Result)"}.Result
   )
   n11 = RigVMVariableNode() {variable=didKill}
   > &RigUnit_SetBoneRotation(Bone=chomp_stem_l_jaw, Space=LocalSpace, bPropagateToChildren=false, Rotation=@n10.Result, Weight=@n11.Value)
   > &RigUnit_SetBoneRotation(
    Bone=chomp_stem_r_jaw
    bPropagateToChildren=True
    Rotation=&RigVMFunction_AccumulateQuatLerp(
     Blend=1
     InitialValue=(X=0,Y=0,Z=0,W=1)
     TargetValue=&RigVMFunction_MathQuaternionSlerp(B=(X=0,Y=0,Z=0,W=1), T=0.2, A=@n10.Result) {template="Interpolate::Execute(in A,in B,in T,out Result)"}.Result
    ) {template="AccumulateLerp::Execute(in TargetValue,in InitialValue,in Blend,in bIntegrateDeltaTime,out Result)"}.Result
    Weight=@n11.Value
   )
   > &RigUnit_SetBoneRotation(Bone=petal_l_mid, bPropagateToChildren=True, Rotation=@n10.Result, Weight=@n11.Value)
   > &RigUnit_SetBoneRotation(Bone=petal_r_mid, bPropagateToChildren=True, Rotation=@n10.Result, Weight=@n11.Value)
   > &RigUnit_SetBoneRotation(Bone=petal_top_mid, bPropagateToChildren=True, Rotation=@n10.Result, Weight=@n11.Value)
  ]}
]
hierarchy [
 root {loc=<0 0 0>, rot=<0 0 0 1>}
 pack {parent=root, loc=<0.0108 2.8551 0>, rot=<0 0 0 1>}
 head_stem {parent=pack, loc=<-3.3927 -6.2486 4.6174>, rot=<0.5 -0.5 0.5 -0.5>}
 head_base_MSTR {parent=head_stem, loc=<3.618 1.8563 3.3942>, rot=<0 0 0 1>}
 head_base {parent=head_base_MSTR, loc=<0.0982 0 0>, rot=<0 0 0 1>}
 tongue_01 {parent=head_base, loc=<1.1549 0.3754 -0.2519>, rot=<0.0029 -0.0295 -0.0967 -0.9949>}
 tongue_02 {parent=tongue_01, loc=<3.2135 0 0>, rot=<-0.0034 0.0224 -0.1512 -0.9882>}
 tongue_03 {parent=tongue_02, loc=<3.9623 0 0>, rot=<0 0 -0.4493 -0.8934>}
 tongue_04 {parent=tongue_03, loc=<4.1009 0 0>, rot=<-0.0046 -0.06 0.0757 -0.9953>}
 tongue_05 {parent=tongue_04, loc=<2.168 0 0>, rot=<0.0331 0.1146 0.2753 -0.9539>}
 tongue_06 {parent=tongue_05, loc=<1.8347 0 0>, rot=<0 0 0.247 -0.969>}
 tongue_07 {parent=tongue_06, loc=<1.1841 0 0>, rot=<0 0 0 -1>}
 petal_top_MSTR {parent=head_base, loc=<-0.0982 -1.0129 0>, rot=<0 0 0.5125 -0.8587>}
 petal_top {parent=petal_top_MSTR, loc=<0.05 0 0>, rot=<0 0 -0.0576 -0.9983>}
 petal_top_mid {parent=petal_top, loc=<4.2027 0 0>, rot=<0 0 0 1>}
 petal_top_tip {parent=petal_top_mid, loc=<4.7389 0 0>, rot=<0 0 0 1>}
 petal_r_MSTR {parent=head_base, loc=<0.2805 0.1138 0.5872>, rot=<0.6661 0.4686 0.3339 -0.4746>}
 petal_r {parent=petal_r_MSTR, loc=<0.05 -0.1102 0>, rot=<0.2042 0.0743 -0.007 -0.9761>}
 petal_r_mid {parent=petal_r, loc=<5.3151 0 0>, rot=<0 0 0 -1>}
 petal_r_tip {parent=petal_r_mid, loc=<4.2393 0 0>, rot=<0 0 0 1>}
 petal_l_MSTR {parent=head_base, loc=<-0.6945 0.2972 -0.609>, rot=<0.7154 0.563 -0.256 0.3252>}
 petal_l {parent=petal_l_MSTR, loc=<0.05 0 0>, rot=<0.0694 0.0513 0.0374 -0.9956>}
 petal_l_mid {parent=petal_l, loc=<4.8009 0 0>, rot=<0 0 0 -1>}
 petal_l_tip {parent=petal_l_mid, loc=<4.6637 0 0>, rot=<0 0 0 -1>}
 stem_btm_01 {parent=pack, loc=<0.1493 -4.4925 -3.3447>, rot=<-0.1839 -0.6217 0.1055 -0.754>}
 stem_btm_02 {parent=stem_btm_01, loc=<4.9047 0 0>, rot=<-0.0245 -0.267 0.0882 -0.9594>}
 stem_btm_03 {parent=stem_btm_02, loc=<6.4669 0 0>, rot=<-0.0573 0.1697 0.0009 -0.9838>}
 stem_btm_05 {parent=stem_btm_03, loc=<6.0362 0 0>, rot=<-0.0296 0.5246 0.0479 -0.8495>}
 stem_btm_06 {parent=stem_btm_05, loc=<7.4425 0 0>, rot=<0.1794 -0.8869 -0.0844 0.4173>}
 stem_btm_04 {parent=stem_btm_06, loc=<8.8793 0 0>, rot=<0 0.8119 0 -0.5839>}
 stem_btm_07 {parent=stem_btm_03, loc=<5.1012 1.1877 -2.7329>, rot=<-0.0736 -0.0557 0.3968 -0.9133>}
 stem_btm_08 {parent=stem_btm_07, loc=<3.1494 0 0>, rot=<0 0.2656 0 -0.9641>}
 stem_l_btm_01 {parent=pack, loc=<-3.2982 -0.586 -2.5448>, rot=<0.0095 -0.649 0.733 -0.2036>}
 stem_l_btm_02 {parent=stem_l_btm_01, loc=<4.8939 0 0>, rot=<0.08 -0.1741 -0.4004 -0.8961>}
 stem_l_btm_03 {parent=stem_l_btm_02, loc=<6.8384 0 0>, rot=<-0.0374 -0.1358 0.8985 -0.4157>}
 stem_l_btm_04 {parent=stem_l_btm_03, loc=<3.5877 0 0>, rot=<0.1224 0.1709 0.6763 -0.706>}
 stem_l_mid_01 {parent=pack, loc=<-6.7517 -3.9665 2.9586>, rot=<-0.4787 -0.5945 0.6445 -0.0461>}
 stem_l_mid_02 {parent=stem_l_mid_01, loc=<5.6401 0 0>, rot=<0.0294 0.1939 0.769 -0.6084>}
 stem_l_mid_03 {parent=stem_l_mid_02, loc=<4.3223 0 0>, rot=<0 0 -0.509 -0.8607>}
 stem_l_mid_04 {parent=stem_l_mid_03, loc=<2.4831 0 0>, rot=<0 0 -0.6455 -0.7638>}
 stem_l_mid_05 {parent=stem_l_mid_04, loc=<4.7057 0 0>, rot=<0 0 0 -1>}
 chomp_stem_l_01 {parent=pack, loc=<-7.4534 -1.8794 6.3837>, rot=<0.6091 -0.4889 0.3759 0.4987>}
 chomp_stem_l_02 {parent=chomp_stem_l_01, loc=<9.2557 0 0>, rot=<-0.0387 -0.1076 -0.5054 -0.8553>}
 chomp_stem_l_03 {parent=chomp_stem_l_02, loc=<2.2091 0 0>, rot=<0 0 0 -1>}
 chomp_stem_l_head {parent=chomp_stem_l_03, loc=<1.0529 0 0>, rot=<0 0 0 -1>}
 chomp_stem_l_jaw {parent=chomp_stem_l_03, loc=<0.7959 0.9512 0>, rot=<0 0 -0.5128 -0.8585>}
 stem_top_mid_01 {parent=pack, loc=<-0.2967 -3.2272 7.1943>, rot=<0.8078 -0.0243 0.5879 -0.0365>}
 stem_top_mid_02 {parent=stem_top_mid_01, loc=<5.1628 0 0>, rot=<-0.009 -0.3056 -0.0279 -0.9517>}
 stem_top_mid_03 {parent=stem_top_mid_02, loc=<6.4819 0 0>, rot=<0 -0.586 0 -0.8103>}
 stem_top_mid_04 {parent=stem_top_mid_03, loc=<3.3673 0 0>, rot=<0 -0.829 0 -0.5592>}
 stem_top_mid_05 {parent=stem_top_mid_04, loc=<4.692 0 0>, rot=<0 0 0 1>}
 stem_top_l_01 {parent=pack, loc=<-3.4285 -2.5083 4.8696>, rot=<0.7708 0.0628 0.6336 0.0224>}
 stem_top_l_02 {parent=stem_top_l_01, loc=<2.8265 0 0>, rot=<0 0.0891 0 -0.996>}
 stem_top_l_03 {parent=stem_top_l_02, loc=<2.4553 0 0>, rot=<0 -0.2411 0 -0.9705>}
 stem_top_r_01 {parent=pack, loc=<6.165 -4.7539 7.541>, rot=<0.7419 0.0898 0.6645 0.005>}
 stem_top_r_02 {parent=stem_top_r_01, loc=<6.141 0 0>, rot=<0 -0.3391 0 -0.9408>}
 stem_top_r_03 {parent=stem_top_r_02, loc=<4.2985 0 0>, rot=<0 -0.2411 0 -0.9705>}
 chomp_stem_r_01 {parent=pack, loc=<4.709 -4.1617 2.7251>, rot=<0.7483 0.0533 0.132 -0.6479>}
 chomp_stem_r_02 {parent=chomp_stem_r_01, loc=<7.7368 0 0>, rot=<0.0206 0.0346 0.5106 -0.8589>}
 chomp_stem_r_03 {parent=chomp_stem_r_02, loc=<4.1146 0 0>, rot=<-0.0569 -0.1369 0.3793 -0.9133>}
 chomp_stem_r_04 {parent=chomp_stem_r_03, loc=<7.0771 0 0>, rot=<0 0 -0.7758 -0.6309>}
 chomp_stem_r_head {parent=chomp_stem_r_04, loc=<1.3852 -0.9461 0.0059>, rot=<0.0031 0 0 -1>}
 chomp_stem_r_jaw {parent=chomp_stem_r_04, loc=<1.3325 -0.0524 0.006>, rot=<0 0 -0.4595 -0.8882>}
 stem_r_01 {parent=pack, loc=<6.0863 -4.6022 -0.85>, rot=<0.6428 -0.3822 0.2013 -0.6326>}
 stem_r_02 {parent=stem_r_01, loc=<3.1148 0 0>, rot=<0.0383 0.2692 -0.1046 -0.9566>}
 stem_r_03 {parent=stem_r_02, loc=<4.5509 0 0>, rot=<0.0643 0.1318 0.4338 -0.889>}
 stem_r_04 {parent=stem_r_03, loc=<2.9435 0 0>, rot=<0 0 0.5881 -0.8088>}
 stem_r_05 {parent=stem_r_02, loc=<2.3301 1.9157 -1.04>, rot=<0.1229 0.2255 0.0639 -0.9643>}
 stem_r_06 {parent=stem_r_05, loc=<4.4399 0 0>, rot=<0 0 0 1>}
 stem_r_btm_01 {parent=pack, loc=<6.7122 -6.6512 -4.0186>, rot=<0.4201 -0.4837 -0.4201 -0.6427>}
 stem_r_btm_02 {parent=stem_r_btm_01, loc=<5.5121 0 0>, rot=<0 0 0.0888 -0.9961>}
 stem_r_btm_03 {parent=stem_r_btm_02, loc=<4.4174 0 0>, rot=<0 -0.0954 0 -0.9954>}
 stem_r_btm_04 {parent=stem_r_btm_02, loc=<5.4438 -0.3949 -2.0048>, rot=<-0.1616 -0.3817 -0.5897 -0.6932>}
 stem_r_btm_05 {parent=stem_r_btm_04, loc=<3.2251 0 0>, rot=<0 0 0 1>}
]
variables [didKill {type=double, default=0}]
