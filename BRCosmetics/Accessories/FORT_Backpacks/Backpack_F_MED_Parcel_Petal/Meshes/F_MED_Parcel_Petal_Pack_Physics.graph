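# Physics asset for the Parcel Petal backpack cosmetic, providing bodies and joints for physics motion of its stems, petals and tongue.
# A box body on the pack roots capsule chains for the stems and chomp stems, plus a head carrying three petals and a six-link tongue.
# Chain joints use limited swing and tight twist, while head and petal master joints rotate freely.
# The stem_r_btm_04 body branches from stem_r_btm_02, not 03.
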PhysicsAsset /BRCosmetics/Accessories/FORT_Backpacks/Backpack_F_MED_Parcel_Petal/Meshes/F_MED_Parcel_Petal_Pack_Physics
bodies {
 chomp_stem_l_01 [{shapeType=Capsule, length=5.0036, radius=2.9572, center=<0.1 0 0>, rotation=<rot 90 0 0>}]
 chomp_stem_l_02 [{shapeType=Capsule, length=2, radius=1, center=<0.05 0 0>, rotation=<rot 90 0 0>}]
 chomp_stem_l_03 [{shapeType=Capsule, length=2, radius=1, center=<0.05 0 0>, rotation=<rot 90 0 0>}]
 chomp_stem_r_01 [{shapeType=Capsule, length=4.4012, radius=2.3466, center=<0.1 0 0>, rotation=<rot 90 0 0>}]
 chomp_stem_r_02 [{shapeType=Capsule, length=2, radius=2, center=<0.01 0 0>, rotation=<rot 90 0 0>}]
 chomp_stem_r_03 [{shapeType=Capsule, length=2, radius=2, center=<0.01 0 0>, rotation=<rot 90 0 0>}]
 head_base [{shapeType=Capsule, length=0, radius=1.0373, center=<0.8825 -0.0883 0.0555>, rotation=<rot 0 -169.5199 -76.4913>}]
 pack [{shapeType=Box, extentX=30, extentY=3, extentZ=26.4841, center=<0 1.1349 -14.8456>, rotation=<rot 0 0 2.2956>}]
 petal_l [{shapeType=Capsule, length=7.0924, radius=4.1041, center=<5 0 0>, rotation=<rot 90 0 0>}]
 petal_l_MSTR [{shapeType=Capsule, length=0.1, radius=0.1}]
 petal_r [{shapeType=Capsule, length=5.6567, radius=4.1, center=<6.7192 0 0>, rotation=<rot 90 0 0>}]
 petal_r_MSTR [{shapeType=Capsule, length=0.1, radius=0.1}]
 petal_top [{shapeType=Capsule, length=7.1696, radius=4.146, center=<6.3155 1.2603 0.0125>, rotation=<rot 0 84.0225 -89.2415>}]
 petal_top_MSTR [{shapeType=Capsule, length=0.505, radius=0.505, rotation=<rot 0 -90 0>}]
 stem_btm_02 [{shapeType=Capsule, length=1, radius=3.6836, center=<0.1 0 0>, rotation=<rot 90 0 0>}]
 stem_btm_03 [{shapeType=Capsule, length=1, radius=4.6666, center=<0.1 0 0>, rotation=<rot 90 0 0>}]
 stem_btm_05 [{shapeType=Capsule, length=1.5372, radius=3.7093, center=<-0.7019 -0.5824 0>, rotation=<rot 90 0 0>}]
 stem_btm_06 [{shapeType=Capsule, length=0.6329, radius=3.6351, center=<1.2589 0 0.0278>, rotation=<rot 23.6072 -0.0001 0>}]
 stem_l_btm_01 [{shapeType=Capsule, length=1.5924, radius=1.9076, center=<0.1 0 0>, rotation=<rot 90 0 0>}]
 stem_l_btm_02 [{shapeType=Capsule, length=0.6999, radius=1.9437, center=<1.8771 0.3041 0>, rotation=<rot 90 0 0>}]
 stem_l_btm_03 [{shapeType=Capsule, length=1.6887, radius=1.8621, center=<1.4114 -0.1473 0>, rotation=<rot 90 0 -16.955>}]
 stem_r_btm_01 [{shapeType=Capsule, length=2, radius=1.005, center=<0.1 0 0>, rotation=<rot 90 0 0>}]
 stem_r_btm_02 [{shapeType=Capsule, length=2, radius=1, center=<0.1 0 0>, rotation=<rot 90 0 0>}]
 stem_r_btm_03 [{shapeType=Capsule, length=2, radius=1, center=<0.1 0 0>, rotation=<rot 90 0 0>}]
 stem_r_btm_04 [{shapeType=Capsule, length=2, radius=1, center=<0.1 0 0>, rotation=<rot 90 0 0>}]
 tongue_01 [{shapeType=Capsule, length=1, radius=1.5, center=<0.075 0 0>, rotation=<rot 90 0 0>}]
 tongue_02 [{shapeType=Capsule, length=1, radius=1.5, center=<0.075 0 0>, rotation=<rot 90 0 0>}]
 tongue_03 [{shapeType=Capsule, length=1, radius=1.5, center=<0.075 0 0>, rotation=<rot 90 0 0>}]
 tongue_04 [{shapeType=Capsule, length=1, radius=1.5, center=<0.075 0 0>, rotation=<rot 90 0 0>}]
 tongue_05 [{shapeType=Capsule, length=1.0167, radius=1.5, center=<0.075 0 0>, rotation=<rot 90 0 0>}]
 tongue_06 [{shapeType=Capsule, length=0.4995, radius=1.5, center=<0.075 0 0>, rotation=<rot 90 0 0>}]
}
constraints [
 {bone1Name=stem_btm_02, bone2Name=pack, swing1LimitDegrees=24.92, swing2LimitDegrees=35, twistLimitDegrees=5}
 {bone1Name=stem_btm_03, bone2Name=stem_btm_02, swing1LimitDegrees=27.8, swing2LimitDegrees=29.24, twistLimitDegrees=5}
 {bone1Name=stem_btm_05, bone2Name=stem_btm_03, swing1LimitDegrees=35, swing2LimitDegrees=35, twistLimitDegrees=5}
 {bone1Name=stem_btm_06, bone2Name=stem_btm_05, swing1LimitDegrees=29.24, swing2LimitDegrees=24.92, twistLimitDegrees=5}
 {
  bone1Name=petal_top_MSTR
  bone2Name=head_base
  swing1LimitDegrees=45
  swing1Motion=Free
  swing2LimitDegrees=45
  swing2Motion=Free
  twistLimitDegrees=45
  twistMotion=Free
 }
 {bone1Name=petal_top, bone2Name=petal_top_MSTR, swing1LimitDegrees=30, swing2LimitDegrees=20, twistLimitDegrees=5}
 {
  bone1Name=petal_r_MSTR
  bone2Name=head_base
  swing1LimitDegrees=45
  swing1Motion=Free
  swing2LimitDegrees=45
  swing2Motion=Free
  twistLimitDegrees=45
  twistMotion=Free
 }
 {bone1Name=petal_r, bone2Name=petal_r_MSTR, swing1LimitDegrees=30, swing2LimitDegrees=30, twistLimitDegrees=5}
 {bone1Name=petal_l, bone2Name=petal_l_MSTR, swing1LimitDegrees=30, swing2LimitDegrees=20, twistLimitDegrees=5}
 {bone1Name=tongue_01, bone2Name=head_base, swing1LimitDegrees=30, swing2LimitDegrees=30, twistLimitDegrees=5}
 {bone1Name=tongue_02, bone2Name=tongue_01, swing1LimitDegrees=30, swing2LimitDegrees=30, twistLimitDegrees=5}
 {bone1Name=tongue_03, bone2Name=tongue_02, swing1LimitDegrees=30, swing2LimitDegrees=30, twistLimitDegrees=5}
 {bone1Name=tongue_04, bone2Name=tongue_03, swing1LimitDegrees=30, swing2LimitDegrees=30, twistLimitDegrees=5}
 {bone1Name=tongue_05, bone2Name=tongue_04, swing1LimitDegrees=30, swing2LimitDegrees=30, twistLimitDegrees=5}
 {bone1Name=tongue_06, bone2Name=tongue_05, swing1LimitDegrees=30, swing2LimitDegrees=30, twistLimitDegrees=5}
 {
  bone1Name=head_base
  bone2Name=pack
  swing1LimitDegrees=45
  swing1Motion=Free
  swing2LimitDegrees=45
  swing2Motion=Free
  twistLimitDegrees=45
  twistMotion=Free
 }
 {
  bone1Name=petal_l_MSTR
  bone2Name=head_base
  swing1LimitDegrees=45
  swing1Motion=Free
  swing2LimitDegrees=45
  swing2Motion=Free
  twistLimitDegrees=45
  twistMotion=Free
 }
 {bone1Name=chomp_stem_l_01, bone2Name=pack, swing1LimitDegrees=25, swing2LimitDegrees=25, twistLimitDegrees=5}
 {bone1Name=chomp_stem_l_02, bone2Name=chomp_stem_l_01, swing1LimitDegrees=25, swing2LimitDegrees=25, twistLimitDegrees=5}
 {bone1Name=chomp_stem_l_03, bone2Name=chomp_stem_l_02, swing1LimitDegrees=25, swing2LimitDegrees=25, twistLimitDegrees=5}
 {bone1Name=chomp_stem_r_01, bone2Name=pack, swing1LimitDegrees=25, swing2LimitDegrees=25, twistLimitDegrees=5}
 {bone1Name=chomp_stem_r_02, bone2Name=chomp_stem_r_01, swing1LimitDegrees=15, swing2LimitDegrees=15, twistLimitDegrees=5}
 {bone1Name=chomp_stem_r_03, bone2Name=chomp_stem_r_02, swing1LimitDegrees=25, swing2LimitDegrees=25, twistLimitDegrees=5}
 {bone1Name=stem_l_btm_01, bone2Name=pack, swing1LimitDegrees=35, swing2LimitDegrees=35, twistLimitDegrees=5}
 {bone1Name=stem_l_btm_02, bone2Name=stem_l_btm_01, swing1LimitDegrees=35, swing2LimitDegrees=35, twistLimitDegrees=5}
 {bone1Name=stem_l_btm_03, bone2Name=stem_l_btm_02, swing1LimitDegrees=35, swing2LimitDegrees=35, twistLimitDegrees=5}
 {bone1Name=stem_r_btm_01, bone2Name=pack, swing1LimitDegrees=35, swing2LimitDegrees=35, twistLimitDegrees=5}
 {bone1Name=stem_r_btm_02, bone2Name=stem_r_btm_01, swing1LimitDegrees=35, swing2LimitDegrees=35, twistLimitDegrees=5}
 {bone1Name=stem_r_btm_03, bone2Name=stem_r_btm_02, swing1LimitDegrees=35, swing2LimitDegrees=35, twistLimitDegrees=5}
 {bone1Name=stem_r_btm_04, bone2Name=stem_r_btm_02, swing1LimitDegrees=35, swing2LimitDegrees=35, twistLimitDegrees=5}
]
