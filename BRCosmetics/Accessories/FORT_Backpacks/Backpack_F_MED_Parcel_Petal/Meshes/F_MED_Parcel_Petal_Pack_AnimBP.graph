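# Animation Blueprint for the Parcel Petal back bling, animating its stems and petals with player movement and reacting to kills.
# RigidBody physics feeds bone offsets for skydiving and floating, weakening in vehicles, shopping carts and on ziplines.
# A watched kill-stat increase starts a per-update timer that fades didKill from one to zero, blending sine-driven petal wobble and driving a final Control Rig.
# The stat-watch async node is invoked from every animation update.
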AnimBlueprint /BRCosmetics/Accessories/FORT_Backpacks/Backpack_F_MED_Parcel_Petal/Meshes/F_MED_Parcel_Petal_Pack_AnimBP
target_skeleton=&/F_MED_Parcel_Petal_Pack_Skeleton
graphs [
 AnimGraph {nodes [
   Component To Local()
   n1 = Get Zipline() -> (Zipline:bool)
   n2 = Get bIsInShoppingCart() -> (bIsInShoppingCart:bool)
   n3 = Get bIsInVehicle() -> (bIsInVehicle:bool)
   n4 = RigidBody(
    ComponentPose=Local To Component(LocalPose=Local Space Ref Pose())
    Alpha=SelectFloat(A=0.25, B=1, bPickA=OR Boolean!(A=@n3, B=@n1, C:bool=@n2) -> (ReturnValue))
   ) {
    bOverrideWorldGravity=true
    overridePhysicsAsset=&/F_MED_Parcel_Petal_Pack_Physics
    simulationSpace=BaseBoneSpace
    alphaBoolBlend {blendOption=Linear}
    baseBoneRef {boneName=root}
    componentLinearAccScale=<0.2 0.2 0.2>
    componentLinearVelScale=<0.1 0.1 0.1>
    overrideWorldGravity {z=-980}
   }
   n5 = "Transform (Modify) Bone - Bone: pack"(ComponentPose=@n4, Translation="0,-2,0", Rotation="0,0,0") {
    rotationSpace=BCS_BoneSpace
    translationMode=BMM_Additive
    alphaBoolBlend {blendOption=Linear}
    boneToModify {boneName=pack}
    translation {y=-2}
   }
   n6 = Get SkyDiving() -> (SkyDiving:double)
   n7 = "Transform (Modify) Bone - Bone: stem_btm_02"(ComponentPose=@n5, Alpha=@n6, Translation="0,0,0", Rotation="0,-40,0", Scale="1,1,1") {
    rotationMode=BMM_Additive
    rotationSpace=BCS_BoneSpace
    alphaBoolBlend {blendOption=Linear}
    boneToModify {boneName=stem_btm_02}
    rotation {yaw=-40}
   }
   n8 = "Transform (Modify) Bone - Bone: stem_l_mid_01"(ComponentPose=@n7, Alpha=@n6, Translation="0,-3,0", Rotation="0,0,45", Scale="1,1,1") {
    rotationMode=BMM_Additive
    translationMode=BMM_Additive
    alphaBoolBlend {blendOption=Linear}
    boneToModify {boneName=stem_l_mid_01}
    rotation {roll=45}
    translation {y=-3}
   }
   n9 = "Transform (Modify) Bone - Bone: stem_l_btm_01"(ComponentPose=@n8, Alpha=@n6, Translation="0,0,0", Rotation="0,30,30", Scale="1,1,1") {rotationMode=BMM_Additive, alphaBoolBlend {blendOption=Linear}, boneToModify {boneName=stem_l_btm_01}, rotation {roll=30, yaw=30}}
   n10 = "Transform (Modify) Bone - Bone: chomp_stem_r_01"(ComponentPose=@n9, Alpha=@n6, Translation="0,0,0", Rotation="0,-45,0", Scale="1,1,1") {rotationMode=BMM_Additive, alphaBoolBlend {blendOption=Linear}, boneToModify {boneName=chomp_stem_r_01}, rotation {yaw=-45}}
   n11 = Get didKill() -> (didKill:double)
   n12 = Get petalRotate() -> (petalRotate:Rotator)
   n13 = "Transform (Modify) Bone - Bone: petal_l_MSTR"(ComponentPose=@n10, Alpha=@n11, Translation="0,0,0", Rotation=@n12, Scale="1,1,1") {rotationMode=BMM_Additive, rotationSpace=BCS_BoneSpace, alphaBoolBlend {blendOption=Linear}, boneToModify {boneName=petal_l_MSTR}}
   n14 = "Transform (Modify) Bone - Bone: petal_r_MSTR"(ComponentPose=@n13, Alpha=@n11, Translation="0,0,0", Rotation=@n12, Scale="1,1,1") {rotationMode=BMM_Additive, rotationSpace=BCS_BoneSpace, alphaBoolBlend {blendOption=Linear}, boneToModify {boneName=petal_r_MSTR}}
   n15 = "Transform (Modify) Bone - Bone: petal_top_MSTR"(ComponentPose=@n14, Alpha=@n11, Translation="0,0,0", Rotation=@n12, Scale="1,1,1") {rotationMode=BMM_Additive, rotationSpace=BCS_BoneSpace, alphaBoolBlend {blendOption=Linear}, boneToModify {boneName=petal_top_MSTR}}
   n16 = "Transform (Modify) Bone - Bone: pack"!(ComponentPose=@n15, bAlphaBoolEnabled:bool=Get bIsSkydiveFloating(), Translation="0,0,0", Rotation="0.000,0.000,10", Scale) -> (Pose) {
    alphaInputType=Bool
    rotationMode=BMM_Additive
    rotationSpace=BCS_BoneSpace
    alphaBoolBlend {blendOption=Linear}
    boneToModify {boneName=pack}
    rotation {roll=10}
   }
   n17 = "Transform (Modify) Bone - Bone: stem_l_mid_01"!(ComponentPose=@n16, bAlphaBoolEnabled:bool=True, Translation="0,0,0", Rotation="25,0.000,0.000", Scale="1,1,1") -> (Pose) {
    alphaInputType=Bool
    rotationMode=BMM_Additive
    rotationSpace=BCS_BoneSpace
    alphaBoolBlend {blendOption=Linear}
    boneToModify {boneName=stem_l_mid_01}
    rotation {pitch=25}
   }
   n18 = Control Rig!(Source=Component To Local(ComponentPose=@n17), Alpha, didKill:double=@n11) -> (Pose) {
    alphaBoolBlend {blendOption=Linear}
    controlRigAssetReference {blueprintRigClass=/BRCosmetics/Accessories/FORT_Backpacks/Backpack_F_MED_Parcel_Petal/Meshes/F_MED_Parcel_Petal_Pack_CtrlRig/F_MED_Parcel_Petal_Pack_CtrlRig.F_MED_Parcel_Petal_Pack_CtrlRig_C}
   }
   Output Pose(Result=@n18) {layerGroup=None}
   Get bIsRiding() -> (bIsRiding:bool)
  ]}
 EventGraph {nodes [
   Event BlueprintUpdateAnimation()
   > Get!(execute:Exec) -> (then:Exec, else:Exec, Pawn:Fort Player Pawn Object Reference)
   > n1 = Sequence!(execute) -> (then_0, then_1, then_2:Exec)
   > n2 = WatchCosmeticStat!(
    execute:Exec
    Pawn:Fort Player Pawn Object Reference=Get Pawn() -> (Pawn:Fort Player Pawn Object Reference)
    StatTag:Fort Stat Manager Tag Structure=(TagName="StatManager.AthenaKills")
    DefaultValue:int=0
    bBroadcastInitialValue:bool=true
    InitialValueBroadcastDelay:float=-1
    ChangeValueBroadcastDelay:float=-1
    bConsumeWaitingNotifications:bool=true
   ) -> (then:Exec, OnStatChanged:Exec, StatName:Name, StatValue:int) {is=title K2Node_AsyncAction}
   n3 = Set KillCount(execute=@n2.OnStatChanged, KillCount:double="To Float (Integer)"(InInt=@n2.StatValue)) -> (Output_Get:double)
   > Branch(Condition="float > float"(A=@n3, B=Get LastKillCount() -> (LastKillCount:double)) {is=title K2Node_CallFunction})
   > Set ActiveKill(ActiveKill:bool=true) -> (Output_Get:bool)
   > Set LastKillCount(LastKillCount:double=@n3) -> (Output_Get:double)
   Branch(execute=@n1.then_1, Condition=Get ActiveKill() -> (ActiveKill:bool))
   > Set timer(timer:double=float + float(A=Get timer() -> (timer:double), B=1) {is=title K2Node_CommutativeAssociativeBinaryOperator}) -> (Output_Get:double)
   n4 = "Sin (Degrees)"!(A:double=float * float(A=GetGameTimeInSeconds(), B=600)) -> (ReturnValue:double)
   > Set petalRotate(petalRotate:Rotator=MakeRotator(Yaw=float * float(A=@n4, B=20))) -> (Output_Get:Rotator)
   n5 = Get timer()
   n6 = Get timerLen() -> (timerLen:double)
   > Set didKill(didKill:double=MapRangeClamped(Value=float - float(A=@n6, B=@n5) {is=title K2Node_CallFunction}, InRangeA=Get timerLen(), InRangeB=0, OutRangeA=1, OutRangeB=0)) -> (Output_Get:double)
   > Branch(Condition="Equal (Float)"!(A:double=Get didKill(), B:double=0) -> (ReturnValue:bool) {is=title K2Node_CallFunction})
   > Set ActiveKill(ActiveKill=false)
   > Set timer(timer=0)
   Set SkyDiving(execute=@n1.then_2, SkyDiving:double=SelectFloat(A=1, B=0, bPickA=Get bIsSkyDiving())) -> (Output_Get:double)
   n7 = Event BlueprintProcessAnimInputEvent()
   > n8 = Branch(Condition=IsNamed(self=@n7.AnimInputEvent, InEventName=aie.Player.Zipline.Entered))
   > Set Zipline(Zipline:bool=true) -> (Output_Get:bool)
   n9 = Branch(execute=@n8.else, Condition=IsNamed(self=@n7.AnimInputEvent, InEventName=aie.Player.Zipline.Exited))
   > Set Zipline(Zipline=false)
   Set Zipline(execute=@n9.else, Zipline=false)
   Event BlueprintInitializeAnimation()
   > n10 = Cast To FortPlayerPawn(Object=TryGetFortPawn())
   > Set Pawn(Pawn:Fort Player Pawn Object Reference=@n10) -> (Output_Get:Fort Player Pawn Object Reference)
   TryGetPawnOwner()
   Set didKill(didKill=0)
  ]}
]
variables [
 ActiveKill {type=boolean}
 didKill {type=number}
 KillCount {type=number}
 LastKillCount {type=number}
 Movement_correction {type=number}
 pawn {type=FortPlayerPawn}
 petalRotate {type=Rotator}
 SkyDiving {type=number}
 timer {type=number}
 timerLen {type=number, default=50}
 Zipline {type=boolean}
]
